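# Material for a hanging holiday bulb on a dark cable, usable with Nanite, with a fake interior glow and gentle sway.
# A high-power view dot product around an emissive center creates the glow, masked by floored UV U and clamped on mobile feature levels.
# World position offset rotates the mesh around a top-of-bounds pivot on a sine, varied by object position, and adjusts normals.
# A bounding-box twist mask is computed but left unconnected.
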Material /Game/Environments/Asteria/Props/Holiday/Holiday_HangingBulb_A/Materials/M_Holiday_HangingBulb_A
graph {
 properties {
  bUseMaterialAttributes=true
  bUsedWithNanite=true
  pixelDepthOffsetMode=PDOM_Legacy
  previewMesh=/Game/Environments/Asteria/Props/Holiday/Holiday_HangingBulb_A/Meshes/SM_HangingBulb_A
  usedShadingModels=Default Lit
 }
 nodes [
  n1 = MaterialFunctionCall() {materialFunction=&BoundingBoxBased_0-1_UVW}
  n2 = ScalarParameter() {defaultValue=0.07, parameterName=TwistMaskPercent, sliderMax=1}
  n3 = Saturate(None=Divide(A=Subtract(A=@n1.B, B=@n2) {constB=0.2}, B=Add(A=@n2)))
  Ceil(None=@n3) {desc=Use to preview mesh mask}
  OneMinus(None=@n3)
  n4 = TransformPosition(None=VectorParameter() {group=Emissive, parameterName=EmissiveCenterPosition, defaultValue=<rgba 0 0 -2 1>}.RGB) {transformType=TRANSFORMPOSSOURCE_World}
  n5 = RotateAboutAxis(NormalizedRotationAxis=NamedRerouteUsage(), RotationAngle=NamedRerouteUsage(), PivotPoint=NamedRerouteUsage(), Position=@n4)
  n6 = CameraVectorWS()
  n7 = ScalarParameter() {defaultValue=-70, group=Emissive, parameterName=EmissiveCenterSoften}
  n8 = Subtract(A=WorldPosition() {worldPositionShaderOffset=WPT_ExcludeAllShaderOffsets}.XYZ, B=Add(A=Add(A=@n4, B=@n5), B=Multiply(A=@n6, B=@n7)))
  n9 = Multiply(A=PixelNormalWS(), B=ScalarParameter() {parameterName=EmissivePixelNormalDistort})
  n10 = DotProduct(A=Normalize(VectorInput=Add(A=@n8, B=@n9)), B=@n6)
  n11 = ScalarParameter() {defaultValue=350, group=Emissive, parameterName=EmissiveFresnel_InterirorFalloff}
  n12 = Multiply(
   A=Multiply(A=Power(Base=@n10, Exp=@n11), B=VectorParameter() {parameterName=Color, defaultValue=<rgba 1 0.7861 0.6225 0>}.RGB)
   B=ScalarParameter() {defaultValue=10, parameterName=Brightness}
  )
  n13 = Clamp(None=@n12, Max=ScalarParameter() {defaultValue=5, parameterName=MobileMaxBrightness})
  n14 = Floor(None=ComponentMask(None=TextureCoordinate()) {r=true})
  n15 = SetMaterialAttributes!(
   MaterialAttributes
   Base Color=VectorParameter() {parameterName=CableColor, defaultValue=<rgba 0.003 0.003 0.003 1>}.RGB
   Metallic=ScalarParameter() {parameterName=Metallic, sliderMax=1}
   Roughness=ScalarParameter() {defaultValue=0.3, parameterName=Roughness, sliderMax=1}
   Emissive Color=Multiply(A=FeatureLevelSwitch(Default=@n12, ES2=@n13, ES3_1=@n13, SM4_REMOVED=@n12), B=Saturate(None=@n14))
  ) {attributeSetTypes [
    #1
    #2
    #3
    #4
   ]}
  n16 = Constant3Vector() {constant=<rgba 0 1 0 0>}
  n17 = NamedRerouteDeclaration(None=Normalize(VectorInput=Transform(None=@n16) {transformSourceType=TRANSFORMSOURCE_Local})) {name=RotationAxis, nodeColor=<rgba 0.3125 1 0 1>}
  n18 = MaterialFunctionCall() {materialFunction=&ObjectPivotPoint}
  n19 = Frac(None=Divide(A=MaterialFunctionCall(f3=@n18.Object Pivot Location) {materialFunction=&AddComponents}.f3) {constB=100})
  n20 = Multiply(
   A=Add(A=Time(), B=@n19)
   B=LinearInterpolate(
    A=ScalarParameter() {defaultValue=0.04, parameterName=RotateSpeed_Min}
    B=ScalarParameter() {defaultValue=0.1, parameterName=RotateSpeed_Max}
    Alpha=@n19
   ) {constA=0.1, constB=0.3}
  )
  n21 = NamedRerouteDeclaration(None=Multiply(A=Sine(None=@n20), B=ScalarParameter() {defaultValue=0.003, parameterName=RotateAmt})) {name=RotationAngle, nodeColor=<rgba 1 0.5859 0 1>}
  n22 = Add(
   A=MaterialFunctionCall() {materialFunction=&ObjectPivotPoint}.Object Pivot Location
   B=Multiply(A=ObjectBounds(), B=Constant3Vector() {constant=<rgba 0 0 1 0>})
  )
  n23 = VertexNormalWS()
  n24 = Add(A=@n23, B=RotateAboutAxis(NormalizedRotationAxis=@n17, RotationAngle=@n21, PivotPoint=Constant3Vector(), Position=@n23))
  SetMaterialAttributes!(
   MaterialAttributes=@n15
   World Position Offset=RotateAboutAxis(
    NormalizedRotationAxis=@n17
    RotationAngle=@n21
    PivotPoint=NamedRerouteDeclaration(None=@n22) {name=RotationPivot, nodeColor=<rgba 1 0.5391 0 1>}
    Position=WorldPosition().XYZ
   )
   Normal=Transform(None=NamedRerouteDeclaration(None=@n24) {name=RotatedNormal, nodeColor=<rgba 0 1 0.5312 1>}) {transformSourceType=TRANSFORMSOURCE_World, transformType=TRANSFORM_Tangent}
  ) {attributeSetTypes [#5, #6]}
 ]
}
